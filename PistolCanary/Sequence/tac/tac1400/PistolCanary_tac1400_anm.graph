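# Level Sequence tac1400_anm for PistolCanary, which animates the tentacle boss.
# It possesses one TentacleBoss Blueprint actor and drives it with a single skeletal animation track holding one section.
# The sequence has no master tracks, and the section names no animation asset here.
# It displays at 60 fps over a 24000 tick resolution.
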LevelSequence /PistolCanary/Sequence/tac/tac1400/PistolCanary_tac1400_anm
bindings [TentacleBoss {
  id=#1
  possessed_class=PistolCanary_TentacleBoss_C
  tracks [{type=MovieSceneSkeletalAnimationTrack, sections [{type=MovieSceneSkeletalAnimationSection, end_frame=3984, start_frame=3599}]}]
 }]
display_rate {denominator=1, numerator=60}
tick_resolution {denominator=1, numerator=24000}
tracks []
