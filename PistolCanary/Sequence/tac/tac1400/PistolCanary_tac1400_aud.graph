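# Level Sequence that handles audio for the tac1400 segment of PistolCanary, centred on the tentacle boss.
# An event trigger track on the possessed AudioManager fires four events, timed at or near the starts of audio sections.
# The possessed TentacleBoss carries its own opening audio section.
# A sequence-level TentacleAudio track then layers four overlapping sections, each with volume, pitch and attach-actor channels.
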LevelSequence /PistolCanary/Sequence/tac/tac1400/PistolCanary_tac1400_aud
bindings [
 PistolCanary_AudioManager {
  id=#1
  possessed_class=PistolCanary_AudioManager_C
  tracks [{
    type=MovieSceneEventTrack
    sections [{
      type=MovieSceneEventTriggerSection
      channels [None {
        type=MovieSceneScriptingEventChannel
        keys [
         [3601, "<Struct 'MovieSceneEvent' (0x852ff398200) {}>"]
         [3743, "<Struct 'MovieSceneEvent' (0x852ff3982a0) {}>"]
         [3796, "<Struct 'MovieSceneEvent' (0x852ff398340) {}>"]
         [3859, "<Struct 'MovieSceneEvent' (0x852ff398480) {}>"]
        ]
       }]
     }]
   }]
 }
 TentacleBoss {
  id=#2
  possessed_class=PistolCanary_TentacleBoss_C
  tracks [{
    type=MovieSceneAudioTrack
    display_name=Audio
    sections [{
      type=MovieSceneAudioSection
      end_frame=3795
      start_frame=3600
      channels [
       Volume {type=MovieSceneScriptingFloatChannel, default=1}
       Pitch {type=MovieSceneScriptingFloatChannel, default=1}
      ]
     }]
   }]
 }
]
display_rate {denominator=1, numerator=60}
tick_resolution {denominator=1, numerator=24000}
tracks [{
  type=MovieSceneAudioTrack
  display_name=TentacleAudio
  sections [
   {
    type=MovieSceneAudioSection
    end_frame=4143
    start_frame=3860
    channels [
     Volume {type=MovieSceneScriptingFloatChannel, default=1}
     Pitch {type=MovieSceneScriptingFloatChannel, default=1}
     AttachActor {type=MovieSceneScriptingActorReferenceChannel}
    ]
   }
   {
    type=MovieSceneAudioSection
    end_frame=3917
    start_frame=3796
    channels [
     Volume {type=MovieSceneScriptingFloatChannel, default=1}
     Pitch {type=MovieSceneScriptingFloatChannel, default=1}
     AttachActor {type=MovieSceneScriptingActorReferenceChannel}
    ]
   }
   {
    type=MovieSceneAudioSection
    end_frame=3857
    start_frame=3736
    channels [
     Volume {type=MovieSceneScriptingFloatChannel, default=1}
     Pitch {type=MovieSceneScriptingFloatChannel, default=1}
     AttachActor {type=MovieSceneScriptingActorReferenceChannel}
    ]
   }
   {
    type=MovieSceneAudioSection
    end_frame=4173
    start_frame=3884
    channels [
     Volume {type=MovieSceneScriptingFloatChannel, default=1}
     Pitch {type=MovieSceneScriptingFloatChannel, default=1}
     AttachActor {type=MovieSceneScriptingActorReferenceChannel}
    ]
   }
  ]
 }]
