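# Gameplay-event Level Sequence for the PistolCanary tentacle boss, named for the tac1400 sequence.
# It possesses a TentacleBoss actor and drives it through a single event trigger track with three closely spaced keys.
# The sequence has no other tracks, and the functions those events call are not shown.
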LevelSequence /PistolCanary/Sequence/tac/tac1400/PistolCanary_tac1400_gam
bindings [TentacleBoss {
  id=#1
  possessed_class=PistolCanary_TentacleBoss_C
  tracks [{
    type=MovieSceneEventTrack
    sections [{
      type=MovieSceneEventTriggerSection
      channels [None {
        type=MovieSceneScriptingEventChannel
        keys [
         [3750, "<Struct 'MovieSceneEvent' (0x852ff3982a0) {}>"]
         [3811, "<Struct 'MovieSceneEvent' (0x852ff398340) {}>"]
         [3874, "<Struct 'MovieSceneEvent' (0x852ff398520) {}>"]
        ]
       }]
     }]
   }]
 }]
display_rate {denominator=1, numerator=60}
tick_resolution {denominator=1, numerator=24000}
tracks []
