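# VFX Level Sequence for shot tac1400, choreographing the tentacle boss's charge-up and laser attack.
# An attached Niagara charge effect ramps through user parameters, then three attached lasers fire in succession with animated rotation, beam width or scale.
# Timed impact and lingering effects at world locations accompany the shots.
# The boss's four body material slots animate emissive color and manually keyed glow pulses, alongside an event trigger on the boss.
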LevelSequence /PistolCanary/Sequence/tac/tac1400/PistolCanary_tac1400_vfx
bindings [
 NiagaraComponent0 {
  id=#1
  parent_id=#2
  possessed_class=NiagaraComponent
  tracks [
   {type=MovieSceneNiagaraSystemTrack, sections [{type=MovieSceneNiagaraSystemSpawnSection, end_frame=4407, start_frame=3807}]}
   {
    type=MovieSceneNiagaraFloatParameterTrack
    display_name=BeamWidth
    sections [{type=MovieSceneFloatSection, channels [None {type=MovieSceneScriptingFloatChannel, default=1, const [1,2]}]}]
   }
   {
    type=MovieSceneParticleTrack
    display_name=FX System
    sections [{type=MovieSceneParticleSection, channels [None {type=MovieSceneScriptingParticleChannel}]}]
   }
  ]
 }
 NS_TentacleAttackImpact {
  id=#3
  possessed_class=NiagaraActor
  tracks [
   {
    type=MovieSceneSpawnTrack
    sections [{type=MovieSceneSpawnSection, channels [None {type=MovieSceneScriptingBoolChannel, default=true, keys [[3763,false],[3800,true]]}]}]
   }
   {
    type=MovieScene3DTransformTrack
    sections [{
      type=MovieScene3DTransformSection
      channels [
       Location.X {type=MovieSceneScriptingDoubleChannel, default=0, const [18108,1]}
       Location.Y {type=MovieSceneScriptingDoubleChannel, default=0, const [76368,1]}
       Location.Z {type=MovieSceneScriptingDoubleChannel, default=0, const [2664,1]}
       Rotation.X {type=MovieSceneScriptingDoubleChannel, default=0, const [0,1]}
       Rotation.Y {type=MovieSceneScriptingDoubleChannel, default=0, const [0,1]}
       Rotation.Z {type=MovieSceneScriptingDoubleChannel, default=0, const [0,1]}
       Scale.X {type=MovieSceneScriptingDoubleChannel, default=1, const [1,1]}
       Scale.Y {type=MovieSceneScriptingDoubleChannel, default=1, const [1,1]}
       Scale.Z {type=MovieSceneScriptingDoubleChannel, default=1, const [1,1]}
      ]
     }]
   }
  ]
 }
 NS_TentacleAttackLaser3 {
  id=#4
  possessed_class=NiagaraActor
  tracks [
   {type=MovieScene3DAttachTrack, sections [{type=MovieScene3DAttachSection, end_frame=12024, start_frame=3767}]}
   {
    type=MovieSceneSpawnTrack
    sections [{type=MovieSceneSpawnSection, channels [None {type=MovieSceneScriptingBoolChannel, default=true, keys [[3773,false],[3863,true]]}]}]
   }
   {
    type=MovieScene3DTransformTrack
    sections [{
      type=MovieScene3DTransformSection
      channels [
       Location.X {type=MovieSceneScriptingDoubleChannel, default=0}
       Location.Y {type=MovieSceneScriptingDoubleChannel, default=0}
       Location.Z {type=MovieSceneScriptingDoubleChannel, default=0}
       Rotation.X {
        type=MovieSceneScriptingDoubleChannel
        default=179.0786
        keys [[3864,16.9847],[3871,13.628],[3874,19.5607],[3877,24.1965],[3879,40.3176],[3977,22.38],[3980,17.8466]]
       }
       Rotation.Y {
        type=MovieSceneScriptingDoubleChannel
        default=-56.8563
        thinned_from=7
        keys [[3864,-67.3622],[3871,-67.3622],[3874,-74.2975],[3879,-82.418],[3977,-79.1153],[3980,-81.4952]]
       }
       Rotation.Z {
        type=MovieSceneScriptingDoubleChannel
        default=-182.8809
        keys [[3864,-14.1507],[3871,-14.1507],[3874,-19.4521],[3877,-27.2327],[3879,-40.103],[3977,692.7673],[3980,698.6074]]
       }
       Scale.X {type=MovieSceneScriptingDoubleChannel, default=15}
       Scale.Y {type=MovieSceneScriptingDoubleChannel, default=15}
       Scale.Z {type=MovieSceneScriptingDoubleChannel, default=25, keys [[3969,15],[3986,25]]}
      ]
     }]
   }
  ]
 }
 TentacleBoss {
  id=#5
  possessed_class=PistolCanary_TentacleBoss_C
  tracks [{
    type=MovieSceneEventTrack
    sections [{
      type=MovieSceneEventTriggerSection
      channels [None {type=MovieSceneScriptingEventChannel, const ["<Struct 'MovieSceneEvent' (0x852ff398520) {}>", 1]}]
     }]
   }]
 }
 NiagaraComponent0 {
  id=#6
  parent_id=#7
  possessed_class=NiagaraComponent
  tracks [
   {type=MovieSceneNiagaraSystemTrack, sections [{type=MovieSceneNiagaraSystemSpawnSection, end_frame=4230, start_frame=3630}]}
   {
    type=MovieSceneNiagaraVectorParameterTrack
    display_name=ChargeScaleVec
    sections [{
      type=MovieSceneFloatVectorSection
      channels [
       Vector.X {type=MovieSceneScriptingFloatChannel, default=0.6, keys [[3632,0],[3637,0.6],[3727,1]]}
       Vector.Y {type=MovieSceneScriptingFloatChannel, default=1, const [1,2]}
       Vector.Z {type=MovieSceneScriptingFloatChannel, default=1, const [1,2]}
      ]
     }]
   }
   {
    type=MovieSceneNiagaraFloatParameterTrack
    display_name=ChargeProgress
    sections [{type=MovieSceneFloatSection, channels [None {type=MovieSceneScriptingFloatChannel, default=0.3, keys [[3631,0],[3663,0.3],[3722,0.5]]}]}]
   }
   {
    type=MovieSceneNiagaraFloatParameterTrack
    display_name=PlasmaEnabled
    sections [{type=MovieSceneFloatSection, channels [None {type=MovieSceneScriptingFloatChannel, default=0, keys [[3731,1],[3740,0]]}]}]
   }
   {
    type=MovieSceneNiagaraColorParameterTrack
    display_name=ChargeScaleColor
    sections [{
      type=MovieSceneColorSection
      channels [
       Color.R {type=MovieSceneScriptingFloatChannel, default=1}
       Color.G {type=MovieSceneScriptingFloatChannel, default=1}
       Color.B {type=MovieSceneScriptingFloatChannel, default=1}
       Color.A {type=MovieSceneScriptingFloatChannel, default=0, keys [[3646,0],[3681,1],[4019,1],[4068,0]]}
      ]
     }]
   }
   {
    type=MovieSceneNiagaraFloatParameterTrack
    display_name=ChargeProgressLightning
    sections [{type=MovieSceneFloatSection, channels [None {type=MovieSceneScriptingFloatChannel, default=0.5472, keys [[3633,0],[3689,0.5472],[3785,1]]}]}]
   }
   {
    type=MovieSceneNiagaraFloatParameterTrack
    display_name=PostProcess
    sections [{type=MovieSceneFloatSection, channels [None {type=MovieSceneScriptingFloatChannel, default=1, thinned_from=3, keys [[3634,0],[3814,1]]}]}]
   }
   {
    type=MovieSceneNiagaraFloatParameterTrack
    display_name=TipDelay
    sections [{type=MovieSceneFloatSection, channels [None {type=MovieSceneScriptingFloatChannel, default=1}]}]
   }
  ]
 }
 NS_TentacleAttackLaser1 {
  id=#8
  possessed_class=NiagaraActor
  tracks [
   {type=MovieScene3DAttachTrack, sections [{type=MovieScene3DAttachSection, end_frame=11985, start_frame=3728}]}
   {
    type=MovieSceneSpawnTrack
    sections [{type=MovieSceneSpawnSection, channels [None {type=MovieSceneScriptingBoolChannel, default=true, keys [[3725,false],[3751,true]]}]}]
   }
   {
    type=MovieScene3DTransformTrack
    sections [{
      type=MovieScene3DTransformSection
      channels [
       Location.X {type=MovieSceneScriptingDoubleChannel, default=0}
       Location.Y {type=MovieSceneScriptingDoubleChannel, default=0}
       Location.Z {type=MovieSceneScriptingDoubleChannel, default=0}
       Rotation.X {type=MovieSceneScriptingDoubleChannel, default=179.0786, keys [[3753,1.3646],[3761,-14.8381]]}
       Rotation.Y {type=MovieSceneScriptingDoubleChannel, default=-56.8563, const [-76.2228,2]}
       Rotation.Z {type=MovieSceneScriptingDoubleChannel, default=-182.8809, const [1.2286,2]}
       Scale.X {type=MovieSceneScriptingDoubleChannel, default=15, const [15,1]}
       Scale.Y {type=MovieSceneScriptingDoubleChannel, default=15, const [15,1]}
       Scale.Z {type=MovieSceneScriptingDoubleChannel, default=30, const [30,1]}
      ]
     }]
   }
  ]
 }
 NS_TentacleAttackImpact3 {
  id=#9
  possessed_class=NiagaraActor
  tracks [
   {
    type=MovieSceneSpawnTrack
    sections [{type=MovieSceneSpawnSection, channels [None {type=MovieSceneScriptingBoolChannel, default=true, keys [[3763,false],[3976,true]]}]}]
   }
   {
    type=MovieScene3DTransformTrack
    sections [{
      type=MovieScene3DTransformSection
      channels [
       Location.X {type=MovieSceneScriptingDoubleChannel, default=0, keys [[3904,17834.3359],[3980,15750.7598]]}
       Location.Y {type=MovieSceneScriptingDoubleChannel, default=0, keys [[3904,83784],[3980,87204]]}
       Location.Z {type=MovieSceneScriptingDoubleChannel, default=0, keys [[3904,2030.3114],[3980,3454.0862]]}
       Rotation.X {type=MovieSceneScriptingDoubleChannel, default=0, const [0,1]}
       Rotation.Y {type=MovieSceneScriptingDoubleChannel, default=0, const [-3.127,1]}
       Rotation.Z {type=MovieSceneScriptingDoubleChannel, default=0, const [0,1]}
       Scale.X {type=MovieSceneScriptingDoubleChannel, default=1, const [1,1]}
       Scale.Y {type=MovieSceneScriptingDoubleChannel, default=1, const [1,1]}
       Scale.Z {type=MovieSceneScriptingDoubleChannel, default=1, const [1,1]}
      ]
     }]
   }
  ]
 }
 NS_TentacleAttackImpact2 {
  id=#10
  possessed_class=NiagaraActor
  tracks [
   {
    type=MovieSceneSpawnTrack
    sections [{type=MovieSceneSpawnSection, channels [None {type=MovieSceneScriptingBoolChannel, default=true, keys [[3763,false],[3893,true]]}]}]
   }
   {
    type=MovieScene3DTransformTrack
    sections [{
      type=MovieScene3DTransformSection
      channels [
       Location.X {type=MovieSceneScriptingDoubleChannel, default=0, const [12902.9717,1]}
       Location.Y {type=MovieSceneScriptingDoubleChannel, default=0, const [80544,1]}
       Location.Z {type=MovieSceneScriptingDoubleChannel, default=0, const [2836.5137,1]}
       Rotation.X {type=MovieSceneScriptingDoubleChannel, default=0, const [0,1]}
       Rotation.Y {type=MovieSceneScriptingDoubleChannel, default=0, const [-3.127,1]}
       Rotation.Z {type=MovieSceneScriptingDoubleChannel, default=0, const [0,1]}
       Scale.X {type=MovieSceneScriptingDoubleChannel, default=1, const [1,1]}
       Scale.Y {type=MovieSceneScriptingDoubleChannel, default=1, const [1,1]}
       Scale.Z {type=MovieSceneScriptingDoubleChannel, default=1, const [1,1]}
      ]
     }]
   }
  ]
 }
 NiagaraComponent0 {
  id=#11
  parent_id=#12
  possessed_class=NiagaraComponent
  tracks [{type=MovieSceneNiagaraSystemTrack, sections [{type=MovieSceneNiagaraSystemSpawnSection, end_frame=4412, start_frame=3812}]}]
 }
 NiagaraComponent0 {
  id=#13
  parent_id=#4
  possessed_class=NiagaraComponent
  tracks [
   {type=MovieSceneNiagaraSystemTrack, sections [{type=MovieSceneNiagaraSystemSpawnSection, end_frame=4463, start_frame=3863}]}
   {
    type=MovieSceneNiagaraFloatParameterTrack
    display_name=BeamWidth
    sections [{type=MovieSceneFloatSection, channels [None {type=MovieSceneScriptingFloatChannel, default=1, keys [[3910,1],[3917,0],[3962,1]]}]}]
   }
   {
    type=MovieSceneParticleTrack
    display_name=FX System
    sections [{type=MovieSceneParticleSection, channels [None {type=MovieSceneScriptingParticleChannel}]}]
   }
  ]
 }
 NiagaraComponent0 {
  id=#14
  parent_id=#8
  possessed_class=NiagaraComponent
  tracks [
   {type=MovieSceneNiagaraSystemTrack, sections [{type=MovieSceneNiagaraSystemSpawnSection, end_frame=4350, start_frame=3750}]}
   {
    type=MovieSceneNiagaraFloatParameterTrack
    display_name=BeamWidth
    sections [{type=MovieSceneFloatSection, channels [None {type=MovieSceneScriptingFloatChannel, default=0, keys [[3775,1],[3782,0]]}]}]
   }
   {
    type=MovieSceneParticleTrack
    display_name=FX System
    sections [{
      type=MovieSceneParticleSection
      channels [None {type=MovieSceneScriptingParticleChannel, keys [[3750, "<ParticleKey.ACTIVATE: 0>"], [3793, "<ParticleKey.DEACTIVATE: 1>"]]}]
     }]
   }
  ]
 }
 NiagaraComponent0 {
  id=#15
  parent_id=#9
  possessed_class=NiagaraComponent
  tracks [{type=MovieSceneNiagaraSystemTrack, sections [{type=MovieSceneNiagaraSystemSpawnSection, end_frame=4576, start_frame=3976}]}]
 }
 NS_TentacleAttackChargeUpLaser {
  id=#7
  possessed_class=NiagaraActor
  tracks [{type=MovieScene3DAttachTrack, sections [{type=MovieScene3DAttachSection, end_frame=11296, start_frame=3608}]}]
 }
 NiagaraComponent0 {
  id=#16
  parent_id=#3
  possessed_class=NiagaraComponent
  tracks [{type=MovieSceneNiagaraSystemTrack, sections [{type=MovieSceneNiagaraSystemSpawnSection, end_frame=4400, start_frame=3800}]}]
 }
 NS_TentacleAttackLingering {
  id=#12
  possessed_class=NiagaraActor
  tracks [{
    type=MovieScene3DTransformTrack
    sections [{
      type=MovieScene3DTransformSection
      channels [
       Location.X {type=MovieSceneScriptingDoubleChannel, default=14264, const [18036,1]}
       Location.Y {type=MovieSceneScriptingDoubleChannel, default=77756, const [76428,1]}
       Location.Z {type=MovieSceneScriptingDoubleChannel, default=3128, const [2652,1]}
       Rotation.X {type=MovieSceneScriptingDoubleChannel, default=0, const [0,1]}
       Rotation.Y {type=MovieSceneScriptingDoubleChannel, default=0, const [-13.4469,1]}
       Rotation.Z {type=MovieSceneScriptingDoubleChannel, default=0, const [0,1]}
       Scale.X {type=MovieSceneScriptingDoubleChannel, default=1, const [1,1]}
       Scale.Y {type=MovieSceneScriptingDoubleChannel, default=1, const [1,1]}
       Scale.Z {type=MovieSceneScriptingDoubleChannel, default=1, const [1,1]}
      ]
     }]
   }]
 }
 NS_TentacleAttackLaser2 {
  id=#2
  possessed_class=NiagaraActor
  tracks [
   {
    type=MovieScene3DTransformTrack
    sections [{
      type=MovieScene3DTransformSection
      channels [
       Location.X {type=MovieSceneScriptingDoubleChannel, default=0}
       Location.Y {type=MovieSceneScriptingDoubleChannel, default=0}
       Location.Z {type=MovieSceneScriptingDoubleChannel, default=0}
       Rotation.X {
        type=MovieSceneScriptingDoubleChannel
        default=179.0786
        keys [[3807,-13.9328],[3808,-52.271],[3810,-8.6304],[3814,-171.4003],[3815,-176.1276],[3817,-176.2859]]
       }
       Rotation.Y {
        type=MovieSceneScriptingDoubleChannel
        default=-56.8563
        keys [[3807,-88.4357],[3808,-87.1223],[3810,-72.0713],[3814,-83.5133],[3815,-78.0567],[3817,-69.1589]]
       }
       Rotation.Z {
        type=MovieSceneScriptingDoubleChannel
        default=-182.8809
        keys [[3807,10.4636],[3808,52.7036],[3810,10.8927],[3814,172.8745],[3815,175.8619],[3817,178.1384]]
       }
       Scale.X {type=MovieSceneScriptingDoubleChannel, default=15}
       Scale.Y {type=MovieSceneScriptingDoubleChannel, default=15}
       Scale.Z {type=MovieSceneScriptingDoubleChannel, default=20, keys [[3807,10],[3815,20]]}
      ]
     }]
   }
   {type=MovieScene3DAttachTrack, sections [{type=MovieScene3DAttachSection, end_frame=12024, start_frame=3767}]}
   {
    type=MovieSceneSpawnTrack
    sections [{type=MovieSceneSpawnSection, channels [None {type=MovieSceneScriptingBoolChannel, default=true, keys [[3773,false],[3807,true]]}]}]
   }
  ]
 }
 SkeletalMesh {
  id=#17
  parent_id=#5
  possessed_class=SkeletalMeshComponent
  tracks [
   {
    type=MovieSceneComponentMaterialTrack
    display_name=Material Slot: Body_A
    sections [{
      type=MovieSceneComponentMaterialParameterSection
      end_frame=4118
      start_frame=3620
      channels [
       GlowPulseEXP {type=MovieSceneScriptingFloatChannel, const [0.015,1]}
       PulseSpeed {type=MovieSceneScriptingFloatChannel, default=0, const [0,1]}
       PulseManualOffset {
        type=MovieSceneScriptingFloatChannel
        default=0
        keys [[3632,0],[3740,1],[3740.56,0],[3793,0],[3802,1],[3803,0],[3880,0],[3889,1],[3890,0],[3961,0],[3970,1],[3971,0]]
       }
       PulseRadius {type=MovieSceneScriptingFloatChannel, const [0.33,1]}
       EmissiveGlowColorR {type=MovieSceneScriptingFloatChannel, const [12,1]}
       EmissiveGlowColorG {type=MovieSceneScriptingFloatChannel, const [2.2855,1]}
       EmissiveGlowColorB {type=MovieSceneScriptingFloatChannel, default=100, keys [[3634,1.2],[3667,100]]}
       EmissiveGlowColorA {type=MovieSceneScriptingFloatChannel, const [1,1]}
      ]
     }]
   }
   {
    type=MovieSceneComponentMaterialTrack
    display_name=Material Slot: Body_B
    sections [{
      type=MovieSceneComponentMaterialParameterSection
      end_frame=4118
      start_frame=3620
      channels [
       PulseManualOffset {
        type=MovieSceneScriptingFloatChannel
        default=0
        keys [[3632,0],[3740,1],[3740.56,0],[3793,0],[3802,1],[3803,0],[3880,0],[3889,1],[3890,0],[3961,0],[3970,1],[3971,0]]
       }
       PulseSpeed {type=MovieSceneScriptingFloatChannel, default=0, const [0,1]}
       EmissiveGlowColorR {type=MovieSceneScriptingFloatChannel, const [12,1]}
       EmissiveGlowColorG {type=MovieSceneScriptingFloatChannel, const [2.2855,1]}
       EmissiveGlowColorB {type=MovieSceneScriptingFloatChannel, default=1.2, keys [[3634,1.2],[3667,100]]}
       EmissiveGlowColorA {type=MovieSceneScriptingFloatChannel, const [1,1]}
      ]
     }]
   }
   {
    type=MovieSceneComponentMaterialTrack
    display_name=Material Slot: Body_C
    sections [{
      type=MovieSceneComponentMaterialParameterSection
      end_frame=4118
      start_frame=3620
      channels [
       PulseSpeed {type=MovieSceneScriptingFloatChannel, default=0, const [0,1]}
       PulseManualOffset {
        type=MovieSceneScriptingFloatChannel
        default=0
        keys [[3632,0],[3740,1],[3740.56,0],[3793,0],[3802,1],[3803,0],[3880,0],[3889,1],[3890,0],[3961,0],[3970,1],[3971,0]]
       }
       EmissiveGlowColorR {type=MovieSceneScriptingFloatChannel, const [12,1]}
       EmissiveGlowColorG {type=MovieSceneScriptingFloatChannel, const [2.2855,1]}
       EmissiveGlowColorB {type=MovieSceneScriptingFloatChannel, default=1.2, keys [[3634,1.2],[3667.1075,100]]}
       EmissiveGlowColorA {type=MovieSceneScriptingFloatChannel, const [1,1]}
      ]
     }]
   }
   {
    type=MovieSceneComponentMaterialTrack
    display_name=Material Slot: Body_D
    sections [{
      type=MovieSceneComponentMaterialParameterSection
      end_frame=4118
      start_frame=3620
      channels [
       PulseManualOffset {
        type=MovieSceneScriptingFloatChannel
        default=0
        keys [[3632,0],[3740,1],[3740.56,0],[3793,0],[3802,1],[3803,0],[3880,0],[3889,1],[3890,0],[3961,0],[3970,1],[3971,0]]
       }
       PulseSpeed {type=MovieSceneScriptingFloatChannel, default=0, const [0,2]}
       EmissiveGlowColorR {type=MovieSceneScriptingFloatChannel, const [12,1]}
       EmissiveGlowColorG {type=MovieSceneScriptingFloatChannel, const [2.2855,1]}
       EmissiveGlowColorB {type=MovieSceneScriptingFloatChannel, default=1.2, keys [[3634,1.2],[3667,100]]}
       EmissiveGlowColorA {type=MovieSceneScriptingFloatChannel, const [1,1]}
      ]
     }]
   }
  ]
 }
 NiagaraComponent0 {
  id=#18
  parent_id=#10
  possessed_class=NiagaraComponent
  tracks [{type=MovieSceneNiagaraSystemTrack, sections [{type=MovieSceneNiagaraSystemSpawnSection, end_frame=4493, start_frame=3893}]}]
 }
]
display_rate {denominator=1, numerator=60}
tick_resolution {denominator=1, numerator=24000}
tracks []
